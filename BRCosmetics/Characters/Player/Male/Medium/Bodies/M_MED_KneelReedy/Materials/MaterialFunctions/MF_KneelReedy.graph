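# Material function for the KneelReedy character that masks its material with a two-ring portal, apparently for an emote.
# Analytic cylinder intersections build the mask from the volume between the rings and both ring discs, gated by camera side.
# Ghost and curve parameters invert the mask, and NoEmote blends it out before it multiplies incoming opacity.
# A ghost-head parameter blends World Position Offset to a small push along vertex normals.
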MaterialFunction /BRCosmetics/Characters/Player/Male/Medium/Bodies/M_MED_KneelReedy/Materials/MaterialFunctions/MF_KneelReedy
graph {nodes [
  n1 = GetMaterialAttributes(None=FunctionInput() {inputName=MA, inputType=FunctionInput_MaterialAttributes}) {attributeGetTypes [#1, #2, #3]}
  n2 = WorldPosition() {worldPositionShaderOffset=WPT_ExcludeAllShaderOffsets}
  n3 = NamedRerouteDeclaration(None=Normalize(VectorInput=VectorParameter() {group=KneelReedy, parameterName=Portal Direction, defaultValue=<rgba 0 0 -1 0>}.RGB)) {name=ZAxis, nodeColor=<rgba 1 0.3516 0 1>}
  n4 = NamedRerouteUsage()
  n5 = MaterialFunctionCall!(
   Origin=NamedRerouteUsage()
   Radius=NamedRerouteUsage()
   Z Axis=NamedRerouteUsage()
   XY Scale
   X Axis
   Capped?=StaticBool() {value=true}
   Z Min-Max=NamedRerouteUsage()
  ) {materialFunction=/Engine/Functions/Engine_MaterialFunctions01/Math/CylinderIntersection}
  n6 = Max(
   A=ComponentMask(None=MaterialFunctionCall!(Vector to Transform=@n2.XYZ, Z Vector=@n3, Center Location=@n4) {materialFunction=/Engine/Functions/Engine_MaterialFunctions02/Math/TransformToZVector}.Result) {b=true}
   B=@n5.Opacity Mask
  )
  n7 = VectorParameter() {group=KneelReedy, parameterName=Ring01Position, defaultValue=<rgba 0 0 50 0>}
  n8 = StaticBool() {value=true}
  n9 = Constant()
  n10 = MaterialFunctionCall(Origin=@n7.RGB, Radius=NamedRerouteUsage(), Z Axis=@n3, Capped?=@n8, Z Min-Max=@n9) {materialFunction=&CylinderIntersection}
  n11 = CameraPositionWS()
  n12 = Clamp(None=ComponentMask(None=MaterialFunctionCall(Vector to Transform=@n11, Z Vector=@n3, Center Location=@n7.RGB) {materialFunction=&TransformToZVector}.Result) {b=true})
  n13 = VectorParameter() {group=KneelReedy, parameterName=Ring02Position, defaultValue=<rgba 0 0 -50 0>}
  n14 = MaterialFunctionCall(Vector to Transform=@n11, Z Vector=@n3, Center Location=@n13.RGB) {materialFunction=&TransformToZVector}
  n15 = OneMinus(None=Clamp(None=ComponentMask(None=@n14.Result) {b=true}))
  n16 = MaterialFunctionCall(Origin=@n13.RGB, Radius=NamedRerouteUsage(), Z Axis=@n3, Capped?=@n8, Z Min-Max=@n9) {materialFunction=&CylinderIntersection}
  n17 = Max(
   A=OneMinus(None=Multiply(A=@n6, B=@n5.Opacity Mask))
   B=Max(A=Multiply(A=@n10.Opacity Mask, B=@n12), B=Multiply(A=@n15, B=@n16.Opacity Mask))
  )
  n18 = LinearInterpolate(A=@n17, B=OneMinus(None=@n17), Alpha=ScalarParameter() {group=KneelReedy, parameterName=Is Ghost?})
  n19 = LinearInterpolate(
   A=LinearInterpolate(
    A=@n18
    B=OneMinus(None=@n18)
    Alpha=ScalarParameter() {desc="0 means inside visible and out side not\r\n\r\n1 means outside visible", group=KneelReedy, parameterName=MaterialCurveB}
   )
   B=@n1.OpacityMask
   Alpha=ScalarParameter() {group=KneelReedy, parameterName=NoEmote}
  )
  n20 = Multiply(A=VertexNormalWS(), B=ScalarParameter() {defaultValue=0.01, group=KneelReedy, parameterName=Ghost Head WPO}) {constB=2}
  n21 = ScalarParameter() {group=KneelReedy, parameterName=Is Ghost Head and Face}
  FunctionOutput(None=SetMaterialAttributes!(
    MaterialAttributes=@n1.MaterialAttributes
    OpacityMask=Multiply(A=@n1.OpacityMask, B=@n19)
    EmissiveColor
    WorldPositionOffset=LinearInterpolate(A=@n1.WorldPositionOffset, B=@n20, Alpha=@n21) {constB=0}
   ) {attributeSetTypes [#2, #1, #3]}) {outputName=Opacity Mask}
  n22 = Abs(None=Subtract(A=ComponentMask(None=@n7.RGB) {b=true}, B=ComponentMask(None=@n13.RGB) {b=true}))
  NamedRerouteDeclaration(None=Multiply(A=AppendVector(A=@n22, B=@n22), B=Constant2Vector() {g=0.5, r=-0.5})) {name=ZMinMax, nodeColor=<rgba 0 0.3672 1 1>}
  NamedRerouteDeclaration(None=Multiply(
    A=ScalarParameter() {defaultValue=100, group=KneelReedy, parameterName=Portal Radius}
    B=ScalarParameter() {defaultValue=1, group=KneelReedy, parameterName=Ring01Scale}
   ) {desc=Ring 01 scaled/animated radius}) {name=ScaledRingRadius, nodeColor=<rgba 0.9922 0 1 1>}
  n23 = Subtract(A=@n7.RGB, B=@n13.RGB)
  NamedRerouteDeclaration(None=Add(A=Multiply(A=@n23) {constB=0.5}, B=@n13.RGB)) {name=AverageRingPos, nodeColor=<rgba 0 1 0.6719 1>}
 ]}
