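# Parent material for the Kneel Reedy player character, combining anime cel shading with an animated flipbook face.
# Flipbook brow, eye and mouth results blend into lit and shaded colors, then pass through eye animation and the anime shader.
# An optional view-dependent silhouette shadow modifies shading; a static switch selects an inverted-hull outline path instead.
# Masked world position offset pulls vertices toward the object origin, and a texture mask drives temporal responsiveness.
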Material /BRCosmetics/Characters/Player/Male/Medium/Bodies/M_MED_KneelReedy/Materials/M_KneelReedy_Parent
graph {
 properties {
  bAutomaticallySetUsageInEditor=false
  bHasPixelAnimation=true
  bUseMaterialAttributes=true
  bUsedWithMeshDeformer=true
  bUsedWithMorphTargets=true
  bUsedWithSkeletalMesh=true
  bUsedWithStaticLighting=true
  pixelDepthOffsetMode=PDOM_Legacy
  shadingModel=MSM_Cloth
  twoSided=true
  usedShadingModels=Cloth
 }
 nodes [
  Composite() {subgraphName=Construct Base Inputs}
  Constant() {r=1}
  LocalPosition()
  n1 = TextureSampleParameter2D() {
   parameterName=Temporal Mask
   samplerType=SAMPLERTYPE_Grayscale
   texture=/BRCosmetics/Characters/Player/Female/Medium/Bodies/F_MED_FearCatch/Textures/T_FearCatch_Head_FX
  }
  Max(A=@n1.R, B=NamedRerouteUsage())
  n2 = TextureObjectParameter() {
   group=Flipbook 02 Brows
   parameterName=FB_BrowTexture_Mask
   samplerType=SAMPLERTYPE_Grayscale
   sortPriority=2
   texture=/BRCosmetics/Characters/Player/Male/Medium/Bodies/M_MED_KneelReedy/Textures/T_BrowState_EyeMask
  }
  NamedRerouteDeclaration(None=@n2) {name=BrowFlipbookTexture, nodeColor=<rgba 1 0.1875 0 1>}
  NamedRerouteDeclaration(None=StaticBoolParameter() {group=Flipbook 02 Brows, parameterName=FB_BrowUseTextureMask, sortPriority=1}) {name=BrowUseTexture, nodeColor=<rgba 0.4531 1 0 1>}
  NamedRerouteDeclaration(None=ScalarParameter() {defaultValue=2, group=Flipbook 02 Brows, parameterName=FB_BrowRowCount_Mask, sortPriority=3}) {name=BrowRowCount, nodeColor=<rgba 1 0 0.7734 1>}
  NamedRerouteDeclaration(None=ScalarParameter() {defaultValue=2, group=Flipbook 02 Brows, parameterName=FB_BrowColumnCount_Mask, sortPriority=4}) {name=BrowColumnCount, nodeColor=<rgba 1 0 0.7734 1>}
  NamedRerouteDeclaration(None=StaticSwitchParameter(True=TextureCoordinate() {coordinateIndex=1}, False=TextureCoordinate()) {defaultValue=true, group=Flipbook 02 Brows, parameterName=FB_BrowUseUV2, sortPriority=10}) {name=BrowUVs, nodeColor=<rgba 1 0.5859 0 1>}
  n3 = ScalarParameter() {defaultValue=1, group=Flipbook 02 Brows, parameterName=FB_BrowUVScaleX, sortPriority=12}
  n4 = ScalarParameter() {defaultValue=1, group=Flipbook 02 Brows, parameterName=FB_BrowUVScaleY, sortPriority=13}
  NamedRerouteDeclaration(None=StaticSwitchParameter(
    True=AppendVector(A=@n3, B=@n4)
    False=ScalarParameter() {defaultValue=1, group=Flipbook 02 Brows, parameterName=FB_BrowUVScale, sortPriority=12}
   ) {group=Flipbook 02 Brows, parameterName=FB_BrowUVNonUniformScale, sortPriority=11}) {name=BrowUVScale, nodeColor=<rgba 0.2656 0 1 1>}
  NamedRerouteDeclaration(None=AppendVector(
    A=ScalarParameter() {group=Flipbook 02 Brows, parameterName=FB_BrowUVOffsetX, sortPriority=20}
    B=ScalarParameter() {group=Flipbook 02 Brows, parameterName=FB_BrowUVOffsetY, sortPriority=21}
   )) {name=BrowUVOffset, nodeColor=<rgba 0.0547 1 0 1>}
  NamedRerouteDeclaration(None=StaticBoolParameter() {group=Flipbook 02 Brows, parameterName=FB_BrowMirrorLeftUV, sortPriority=30}) {name=BrowMirrorLeftUV, nodeColor=<rgba 0 0.6953 1 1>}
  n5 = MaterialFunctionCall(Float2=ComponentMask(None=TextureCoordinate() {coordinateIndex=1}) {g=true, r=true}) {materialFunction=&BreakOutFloat2Components}
  n6 = OneMinus(None=@n5.R)
  n7 = Saturate(None=NamedRerouteUsage())
  n8 = MaterialFunctionCall!(Mouth UV=AppendVector(A=LinearInterpolate(A=@n5.R, B=@n6, Alpha=@n7), B=@n5.G), Eye UV, Brow UV, Material) {materialFunction=/Game/Global/Materials/Character_Parent_2023/FlipbookFace_Parent_2023/MaterialFunctions/MFI_FlipbookFace_Anime}
  NamedRerouteDeclaration(None=@n8.EyesColor) {name=Eyes Color, nodeColor=<rgba 0.0312 1 0 1>}
  NamedRerouteDeclaration(None=ChannelMaskParameter(None=NamedRerouteUsage()) {group=Flipbook 02 Brows, parameterName=FB_BrowVertexColorChannel, sortPriority=0}.R) {name=BrowVertexColor, nodeColor=<rgba 1 0 0.9375 1>}
  NamedRerouteDeclaration(None=StaticBoolParameter() {defaultValue=true, group=Flipbook 02 Brows, parameterName=FB_BrowUseRemap_Mask, sortPriority=190}) {name=BrowUseRemap, nodeColor=<rgba 0 1 0.7422 1>}
  NamedRerouteDeclaration(None=StaticBoolParameter() {group=Flipbook 02 Brows, parameterName=FB_BrowRemapIdleOnly, sortPriority=200}) {name=BrowRemapIdleOnly, nodeColor=<rgba 0 1 0.7422 1>}
  NamedRerouteDeclaration(None=StaticBoolParameter() {group=Flipbook 02 Brows, parameterName=FB_BrowRemapPoseSync, sortPriority=191}) {name=BrowUseRemapPoseSync, nodeColor=<rgba 0 1 0.7422 1>}
  NamedRerouteDeclaration(None=StaticBoolParameter() {defaultValue=true, group=Flipbook 01 Global, parameterName=FB_UseBrows, sortPriority=-10}) {name=UseBrows, nodeColor=<rgba 0.1953 1 0 1>}
  n9 = VectorParameter() {
   group=Flipbook 02 Brows
   parameterName=FB_BrowRemap_0-3 - Mask
   sortPriority=201
   channelNames {a=3 - Brow Dn Glabella Up, b=2 - Brow Up Glabella Dn, g=1 - Brow Up Glabella Up, r=0 - Brow Neutral}
   defaultValue=<rgba 0 1 2 3>
  }
  NamedRerouteDeclaration(None=AppendVector(A=@n9, B=@n9)) {name=BrowRemap_0-3, nodeColor=<rgba 1 0 0.3984 1>}
  n10 = VectorParameter() {
   group=Flipbook 02 Brows
   parameterName=FB_BrowRemap_4-7 - Mask
   sortPriority=202
   channelNames {a=7 - Glabella Up, b=6 - Brow Down, g=5 - Brow Up, r=4 - Brow Dn Glabella Dn}
   defaultValue=<rgba 4 5 6 7>
  }
  NamedRerouteDeclaration(None=AppendVector(A=@n10, B=@n10)) {name=BrowRemap_4-7, nodeColor=<rgba 1 0 0.3984 1>}
  n11 = VectorParameter() {
   group=Flipbook 02 Brows
   parameterName=FB_BrowRemap_8-11 - Mask
   sortPriority=203
   channelNames {a=Unused, b=Unused, g=Unused, r=8 - Glabella Dn}
   defaultValue=<rgba 8 9 10 11>
  }
  NamedRerouteDeclaration(None=AppendVector(A=@n11, B=@n11)) {name=BrowRemap_8-11, nodeColor=<rgba 1 0 0.3984 1>}
  NamedRerouteDeclaration(None=VertexColor()) {name=VertexColor, nodeColor=<rgba 0.6172 1 0 1>}
  n12 = ScalarParameter() {
   desc="Driven by Animation Blueprint.  Do not set manually."
   group=Flipbook 05 Indices
   parameterName=Flipbook_face_R_brow_index
   sortPriority=1
  }
  n13 = ScalarParameter() {
   desc="Driven by Animation Blueprint.  Do not set manually."
   group=Flipbook 05 Indices
   parameterName=Flipbook_face_L_brow_index
   sortPriority=0
  }
  NamedRerouteDeclaration(None=MaterialFunctionCall!(
    IndexRight=@n12
    IndexLeft=@n13
    LeftRightSync=StaticBoolParameter() {group=Flipbook 02 Brows, parameterName=FB_BrowsSyncLeftRight, sortPriority=40}
    UseLeftIndexForSync=StaticBoolParameter() {group=Flipbook 02 Brows, parameterName=FB_BrowsUseLeftIndexForSync, sortPriority=41}
    "--------------------------      "
    VertexColor=NamedRerouteUsage()
    ----------------------------
    Use Texture Flipbook=NamedRerouteUsage()
    Flipbook Texture=NamedRerouteUsage()
    Flipbook Column Count=NamedRerouteUsage()
    Flipbook Row Count=NamedRerouteUsage()
    UVs=NamedRerouteUsage()
    UV Scale=NamedRerouteUsage()
    UV Offset=NamedRerouteUsage()
    MirrorLeftUV=NamedRerouteUsage()
    -----------------------------
    Use Remap=NamedRerouteUsage()
    Remap Idle Only=NamedRerouteUsage()
    RemapPoseSync=NamedRerouteUsage()
    RemapIndex_0-3=NamedRerouteUsage()
    RemapIndex_4-7=NamedRerouteUsage()
    RemapIndex_8-11=NamedRerouteUsage()
    RemapIndex_12-15
    Remap Use 12 Inputs=StaticBool() {value=true}
    Remap Use 16 Inputs
   ) {materialFunction=/Game/Global/Materials/Character_Parent_2023/FlipbookFace_Parent_2023/MaterialFunctions/Common/MF_Flipbook}.FlipbookTexture) {name=Brow Mask, nodeColor=<rgba 1 0 0.9375 1>}
  n14 = TransformPosition(None=VectorParameter() {parameterName=Head Position}.RGB) {transformType=TRANSFORMPOSSOURCE_TranslatedWorld}
  n15 = Transform(None=Normalize(VectorInput=Multiply(A=@n14) {constB=-1})) {transformSourceType=TRANSFORMSOURCE_World, transformType=TRANSFORM_Local}
  n16 = DotProduct(A=VectorParameter() {parameterName=HeadRightVector, defaultValue=<rgba 1 0 0 0>}.RGB, B=@n15)
  n17 = DotProduct(A=VectorParameter() {parameterName=HeadForwardVector, defaultValue=<rgba 0 1 0 0>}.RGB, B=@n15)
  n18 = LinearInterpolate(A=Saturate(None=Sign(None=@n16)), B=ConstantBiasScale(None=@n16), Alpha=Saturate(None=Sign(None=@n17)))
  NamedRerouteDeclaration(None=Sign(None=Subtract(A=@n18) {constB=0.5})) {name=LeftRightSign, nodeColor=<rgba 1 0 0.1172 1>}
  n19 = Time() {bOverride_Period=true, period=16}
  n20 = ScalarParameter() {parameterName=flipbook_face_mouth_index, sliderMax=15}
  NamedRerouteDeclaration(None=StaticSwitchParameter(True=Floor(None=@n19), False=Floor(None=@n20)) {parameterName=Debug_CycleMouthIndex}) {name=Mouth Index, nodeColor=<rgba 0 1 0.7656 1>}
  NamedRerouteDeclaration(None=@n8.EyesAlpha) {name=Eyes Alpha, nodeColor=<rgba 0.9922 1 0 1>}
  n21 = NamedRerouteUsage()
  n22 = Constant() {r=1}
  n23 = NamedRerouteUsage()
  n24 = Step(Y=OneMinus(None=Desaturation(None=@n23)), X=ScalarParameter() {defaultValue=0.1, group=Kneel Reedy - Character, parameterName=Eye Step})
  n25 = Step(Y=ScalarParameter() {defaultValue=0.9, group=Kneel Reedy - Character, parameterName=Eye and Mouth Step}, X=NamedRerouteUsage()) {constX=0.01, constY=0.9}
  NamedRerouteDeclaration(None=Multiply(A=StaticSwitchParameter(True=@n21, False=@n22) {parameterName=Reworked}, B=Multiply(A=@n24, B=@n25))) {name=Eye Mask, nodeColor=<rgba 0.6875 0 1 1>}
  NamedRerouteDeclaration(None=@n8.BrowsVertexMask) {name=Brow Vertex Mask, nodeColor=<rgba 1 0.6562 0 1>}
  NamedRerouteDeclaration(None=@n8.MouthAlpha) {name=Mouth Alpha, nodeColor=<rgba 0.9219 0 1 1>}
  NamedRerouteUsage()
  PinBase()
  n26 = Constant()
  n27 = PreSkinnedPosition()
  n28 = ShaderStageSwitch(PixelShader=VertexInterpolator(VS=@n27).PS, VertexShader=@n27)
  n29 = MaterialFunctionCall!(Material) {materialFunction=/Game/Packages/Fortress_SharedMaterials/Base_Material_Functions/CharacterBrightnessandScale_Balance}
  n30 = Fresnel() {baseReflectFraction=0, exponent=1}
  PinBase!(M=@n26, Pre Skinned Local Position=@n28, Distance Blend=@n29.Distance Blend, Fresnel=@n30) {reroutePins [
    M {expression=&self:MaterialExpressionReroute_44}
    Pre Skinned Local Position {expression=&self:MaterialExpressionReroute_45}
    Distance Blend {expression=&self:MaterialExpressionReroute_46}
    Fresnel {expression=&self:MaterialExpressionReroute_47}
   ]}
  n31 = MaterialFunctionCall() {materialFunction=/Game/Global/Materials/Character_Parent_2023/Anime_Parent_2023/MaterialFunctions/MFC_AnimeOutline_InvHull_2023}
  MaterialExpressionReroute_44 = Reroute(None=@n26)
  MaterialExpressionReroute_45 = Reroute(None=@n28)
  MaterialExpressionReroute_46 = Reroute(None=@n29.Distance Blend)
  MaterialExpressionReroute_47 = Reroute(None=@n30)
  n32 = MaterialFunctionCall(
   Material=MaterialFunctionCall!(Material=@n31.Result, FX=@n31.Result) {materialFunction=/Game/Packages/Fortress_SharedMaterials/Anime_Master_Material/MFC_ClothWind}.Result
   M=@MaterialExpressionReroute_44.M
   Pre Skinned Local Position=@MaterialExpressionReroute_45.Pre Skinned Local Position
   Distance Blend=@MaterialExpressionReroute_46.Distance Blend
   Fresnel=@MaterialExpressionReroute_47.Fresnel
  ) {materialFunction=&MF_CharacterParent_2023_PostFX}
  n33 = LocalPosition()
  n34 = VertexNormalWS()
  n35 = Constant3Vector() {constant=<rgba 0 0 -1 0>}
  n36 = Multiply(
   A=Add(A=ComponentMask(None=@n33.XYZ) {g=true}, B=ScalarParameter() {defaultValue=-1.5, group=Kneel Reedy - Character, parameterName=Fringe Mask})
   B=Saturate(None=DotProduct(A=@n34, B=@n35))
  )
  n37 = VectorParameter() {group=Kneel Reedy - Character, parameterName=View Mask Depth, defaultValue=<rgba 0 0 -2 1>}
  n38 = Fresnel(Normal=LinearInterpolate(
    A=PixelNormalWS()
    B=Transform(None=@n37.RGB) {transformSourceType=TRANSFORMSOURCE_View}
    Alpha=ScalarParameter() {defaultValue=0.4688, group=Kneel Reedy - Character, parameterName=Blend View Normals}
   )) {baseReflectFraction=0, exponent=1}
  n39 = VertexNormalWS()
  n40 = DotProduct(A=LocalPosition().XYZ, B=Constant3Vector() {constant=<rgba -0.5 0.5 0 0>})
  n41 = Multiply(
   A=DotProduct(
    A=Transform(None=@n39) {transformSourceType=TRANSFORMSOURCE_World, transformType=TRANSFORM_Local}
    B=Constant3Vector() {constant=<rgba 0.5 1 0.15 0>}
   )
   B=Saturate(None=Add(A=@n40) {constB=6})
  )
  n42 = Multiply(A=OneMinus(None=Step(Y=@n38) {constX=0.88}), B=Step(X=Saturate(None=@n41)) {constX=0.01, constY=0.01})
  n43 = TextureSample() {samplerType=SAMPLERTYPE_Normal, texture=/BRCosmetics/Characters/Player/Male/Medium/Bodies/M_MED_KneelReedy/Textures/T_KneelReedy_TestTangent}
  n44 = Transform(None=CameraVectorWS()) {transformSourceType=TRANSFORMSOURCE_World, transformType=TRANSFORM_Tangent}
  n45 = ScalarParameter() {group=Kneel Reedy - Character, parameterName=Blend Screenpace Normal}
  n46 = LinearInterpolate(
   A=Transform(None=LinearInterpolate(A=@n43.RGB, B=@n44, Alpha=@n45)) {transformType=TRANSFORM_Local}
   B=Transform(None=VertexNormalWS()) {transformSourceType=TRANSFORMSOURCE_World, transformType=TRANSFORM_Local}
   Alpha=ScalarParameter() {defaultValue=1, group=Kneel Reedy - Character, parameterName=Blend VS Normals}
  )
  n47 = Transform(None=CameraVectorWS()) {transformSourceType=TRANSFORMSOURCE_World, transformType=TRANSFORM_Local}
  n48 = LinearInterpolate(
   A=ScalarParameter() {defaultValue=0.2, group=Kneel Reedy - Character, parameterName=Silhouette Shadow - Step Far}
   B=ScalarParameter() {defaultValue=0.45, group=Kneel Reedy - Character, parameterName=Silhouette Shadow - Step Near}
   Alpha=MaterialFunctionCall(
    Fade Length=ScalarParameter() {defaultValue=50, group=Kneel Reedy - Character, parameterName=Silhouette Shadow - Fade Length}
    Fade Offset=ScalarParameter() {defaultValue=50, group=Kneel Reedy - Character, parameterName=Silhouette Shadow - Fade Offset}
   ) {materialFunction=&CameraDepthFade}.Result
  )
  n49 = VertexNormalWS()
  n50 = MaterialFunctionCall(
   Axis=Transform(None=VectorParameter() {group=Kneel Reedy - Character, parameterName=Aligned Axis, defaultValue=<rgba 0 0.1 1 1>}.RGB) {transformSourceType=TRANSFORMSOURCE_World, transformType=TRANSFORM_Local}
   Exponent=ScalarParameter() {defaultValue=1.15, group=Kneel Reedy - Character, parameterName=Aligned Exp}
   Normal=@n49
  ) {materialFunction=&AxisAlignedFresnel}
  n51 = ScalarParameter() {defaultValue=1, group=Kneel Reedy - Character, parameterName=Axis Aligned Strength}
  n52 = ScalarParameter() {defaultValue=1.2, group=Kneel Reedy - Character, parameterName=Step Multipler}
  n53 = Min(A=Step(Y=DotProduct(A=@n46, B=@n47), X=@n48), B=Step(Y=Multiply(A=@n50.Fresnel, B=@n51), X=Multiply(A=@n48, B=@n52)))
  n54 = Saturate(None=Multiply(A=OneMinus(None=@n36), B=Max(A=@n42, B=@n53)))
  n55 = MaterialFunctionCall!(CustomUV) {materialFunction=/Game/Global/Materials/Character_Parent_2023/Anime_Parent_2023/MaterialFunctions/MFC_AnimeShader_2023_Pre}
  n56 = MaterialFunctionCall!(CustomLightVector, OverrideNormal_WS, STT=@n55.STT, Normal_Map=@n55.STT, DFL=@n55.STT, FX=@n55.STT) {materialFunction=/Game/Global/Materials/Character_Parent_2023/Anime_Parent_2023/MaterialFunctions/MFC_AnimeShader_2023_Calculations}
  n57 = NamedRerouteUsage()
  n58 = StaticSwitchParameter(True=Multiply(A=NamedRerouteUsage(), B=@n8.FaceAlpha), False=@n8.FaceAlpha) {parameterName=Reworked}
  n59 = ScalarParameter() {defaultValue=1, parameterName=FlipbookFaceAnimeShadedValue}
  n60 = MaterialFunctionCall!(
   Color Lit=LinearInterpolate(A=@n55.Color_Lit, B=@n8.FaceColor, Alpha=@n58)
   Color Shaded=LinearInterpolate(A=@n55.Color_Shaded, B=Multiply(A=@n8.FaceColor, B=@n59), Alpha=@n58)
   Use UV2=StaticBoolParameter() {defaultValue=true, group=01_EyeAnimation_Parameters, parameterName=Use UV2 for Eye Animation, sortPriority=-30}
   Mask=NamedRerouteUsage()
  ) {materialFunction=/BRCosmetics/Characters/Player/Male/Medium/Bodies/M_MED_KneelReedy/Materials/MaterialFunctions/MF_AnimeEyeAnimation_v2_KneelReedy}
  n61 = MaterialFunctionCall!(
   Shading_Gradient=StaticSwitchParameter(True=LinearInterpolate(A=@n54, B=@n56.Shading_Gradient, Alpha=@n57), False=@n56.Shading_Gradient) {parameterName=Use Silhouette Shadow}
   Threshold_Gradient=@n56.Threshold_Gradient
   Specular_Gradient=@n56.Threshold_Gradient
   Sheen_Mask=@n56.Threshold_Gradient
   Thickness=@n56.Threshold_Gradient
   DFL_Mask=@n56.Threshold_Gradient
   FX=@n56.Threshold_Gradient
   Metallic_Gradient=@n56.Threshold_Gradient
   RimLight_Gradient=@n56.Threshold_Gradient
   NormalBuffer=@n56.Threshold_Gradient
   --------
   CINE_LightSources=@n56.Threshold_Gradient
   CINE_LightColor=@n56.Threshold_Gradient
   ---------
   Color_Lit=@n60.Lit
   Color_Shaded=@n60.Lit
   SRM=@n55.SRM
   Emissive=@n55.SRM
   DFL_Color=@n55.SRM
   Specular_Color=@n55.SRM
   RimLight_Color=@n55.SRM
  ) {materialFunction=/Game/Global/Materials/Character_Parent_2023/Anime_Parent_2023/MaterialFunctions/MFC_AnimeShader_2023_Post}
  n62 = GetMaterialAttributes(None=MaterialFunctionCall!(MA=StaticSwitchParameter(True=@n32.Result, False=@n61.Result) {group=Kneel Reedy - Character, parameterName=Use Anime Outline Inv Hull}) {materialFunction=/BRCosmetics/Characters/Player/Male/Medium/Bodies/M_MED_KneelReedy/Materials/MaterialFunctions/MF_KneelReedy}.Opacity Mask) {attributeGetTypes [#1]}
  n63 = ObjectPositionWS()
  n64 = WorldPosition() {worldPositionShaderOffset=WPT_CameraRelativeNoOffsets}
  n65 = NamedRerouteUsage()
  n66 = VertexColor()
  SetMaterialAttributes!(
   MaterialAttributes=@n62.MaterialAttributes
   World Position Offset=LinearInterpolate(A=Subtract(A=@n63, B=@n64.XYZ), B=@n62.World Position Offset, Alpha=Max(A=@n65, B=@n66.G))
  ) {attributeSetTypes [#1]}
  n67 = VertexColor()
  n68 = AppendVector(A=@n67, B=@n67)
  n69 = TextureSampleParameter2D(UVs=MaterialFunctionCall(UVs=TextureCoordinate() {coordinateIndex=1}, Texture Scale=ScalarParameter() {defaultValue=0.464, parameterName=Eye Mask Tiling}) {materialFunction=&ScaleUVsByCenter}.UVs) {
   group=01_EyeAnimation_Parameters
   parameterName=EyeMask
   samplerSource=SSM_Wrap_WorldGroupSettings
   samplerType=SAMPLERTYPE_Masks
   sortPriority=-32
   texture=&T_WhiteMask
  }
  n70 = Subtract(A=ComponentMask(None=TextureCoordinate() {coordinateIndex=1}) {r=true}) {constB=0.0001}
  StaticSwitchParameter(
   True=ChannelMaskParameter(None=StaticSwitchParameter(True=@n68, False=@n69.RGBA) {group=01_EyeAnimation_Parameters, parameterName=UseVertexColorForEyeMask, sortPriority=-33}) {group=01_EyeAnimation_Parameters, parameterName=EyeMaskChannel, sortPriority=-32}.R
   False=Saturate(None=Sign(None=@n70))
  ) {group=01_EyeAnimation_Parameters, parameterName=UseEyeMask, sortPriority=-33}
  TemporalResponsivenessOutput!(None=LinearInterpolate(B=ScalarParameter() {parameterName=Temporal Responsiveness}, Alpha=@n1.R) {constA=0.4})
  Transform(None=@n49) {transformSourceType=TRANSFORMSOURCE_World, transformType=TRANSFORM_Local}
  Transform() {transformType=TRANSFORM_View}
 ]
}
