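# Character-part Anim Blueprint for KneelReedy's face accessory, adding secondary motion to hair strands over a pose copied from the parent mesh.
# Two side-and-front Anim Dynamics chains with differing joint offsets blend by bool, then feed top-strand dynamics.
# The event graph computes a looseHair flag from movement states and a gravity override from a parameters asset.
# It also drives a dynamic outline overlay material with pelvis position and animation curves.
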AnimBlueprint /BRCosmetics/Characters/Player/Male/Medium/Bodies/M_MED_KneelReedy/Meshes/Parts/M_MED_KneelReedy_FaceAcc_AnimBP
target_skeleton=&/M_MED_KneelReedy_FaceAcc_Skeleton
graphs [
 AnimGraph {nodes [
   Output Pose(Result=Use cached pose 'DynamicsCache'()) {layerGroup=None}
   Save cached pose 'CRCache_Input'(Pose=Copy Pose From Mesh() {bCopyCurves=true, bUseAttachedParent=true}) {cacheName=CRCache_Input}
   n1 = Anim Dynamics - Bone: dyn_hairStrand_Side_A_01!(ComponentPose=Local To Component(LocalPose=Use cached pose 'CRCache_Input'()), Alpha=1, GravityOverride:Vector="0,0,-980") -> (Pose) {
    angularDampingOverride=1
    angularSpringConstant=10000
    bAngularSpring=true
    bOverrideAngularDamping=true
    bUseGravityOverride=true
    gravityScale=0.35
    simulationSpace=RootRelative
    alphaBoolBlend {blendOption=Linear}
    boundBone {boneName=dyn_hairStrand_Side_A_01}
    componentAppliedLinearAccClamp=<7500 7500 7500>
    componentLinearAccScale=<1 1 1>
    componentLinearVelScale=<1 1 1>
    gravityOverride {z=-980}
    physicsBodyDefinitions [{
      boundBone {boneName=dyn_hairStrand_Side_A_01}
      boxExtents=<16 22 22>
      constraintSetup {angularLimitsMax=<0 10 25>, angularLimitsMin=<0 -15 -25>, angularTarget=<1 0 0>}
      localJointOffset=<0.5 0 0>
     }]
    simSpaceSettings {maxAngularAcceleration=100, maxAngularVelocity=20, simSpaceAngularAlpha=1}
   }
   n2 = Anim Dynamics - Bone: dyn_hairStrand_Side_B_01!(ComponentPose=@n1, Alpha=1, GravityOverride:Vector="0,0,-980") -> (Pose) {
    angularDampingOverride=1
    angularSpringConstant=10000
    bAngularSpring=true
    bOverrideAngularDamping=true
    bUseGravityOverride=true
    gravityScale=0.35
    simulationSpace=RootRelative
    alphaBoolBlend {blendOption=Linear}
    boundBone {boneName=dyn_hairStrand_Side_B_01}
    componentAppliedLinearAccClamp=<7500 7500 7500>
    componentLinearAccScale=<1 1 1>
    componentLinearVelScale=<1 1 1>
    gravityOverride {z=-980}
    physicsBodyDefinitions [{
      boundBone {boneName=dyn_hairStrand_Side_B_01}
      boxExtents=<18 22 22>
      constraintSetup {angularLimitsMax=<0 20 25>, angularLimitsMin=<0 -20 -25>, angularTarget=<1 0 0>}
      localJointOffset=<0.75 0 0>
     }]
    simSpaceSettings {maxAngularAcceleration=100, maxAngularVelocity=20, simSpaceAngularAlpha=1}
   }
   n3 = Anim Dynamics - Bone: dyn_hairStrand_Front_01!(ComponentPose=@n2, Alpha=1, GravityOverride:Vector="0,0,-980") -> (Pose) {
    angularDampingOverride=1
    angularSpringConstant=1000
    bAngularSpring=true
    bOverrideAngularDamping=true
    bUseGravityOverride=true
    gravityScale=0.35
    simulationSpace=RootRelative
    alphaBoolBlend {blendOption=Linear}
    boundBone {boneName=dyn_hairStrand_Front_01}
    componentAppliedLinearAccClamp=<3000 3000 3000>
    componentLinearAccScale=<1 1 1>
    componentLinearVelScale=<1 1 1>
    gravityOverride {z=-980}
    physicsBodyDefinitions [{
      boundBone {boneName=dyn_hairStrand_Front_01}
      boxExtents=<12 16 16>
      constraintSetup {angularLimitsMax=<0 40 25>, angularLimitsMin=<0 -20 -25>, angularTarget=<1 0 0>}
      localJointOffset=<0.5 0 0>
     }]
    simSpaceSettings {maxAngularAcceleration=100, maxAngularVelocity=20, simSpaceAngularAlpha=1}
   }
   n4 = Component To Local(ComponentPose=@n3)
   n5 = Anim Dynamics - Bone: dyn_hairStrand_Side_A_01(ComponentPose=Local To Component(LocalPose=Use cached pose 'CRCache_Input'()), Alpha=1, GravityOverride="0,0,-980") {
    angularDampingOverride=1
    angularSpringConstant=10000
    bAngularSpring=true
    bOverrideAngularDamping=true
    bUseGravityOverride=true
    gravityScale=0.35
    simulationSpace=RootRelative
    alphaBoolBlend {blendOption=Linear}
    boundBone {boneName=dyn_hairStrand_Side_A_01}
    componentAppliedLinearAccClamp=<7500 7500 7500>
    componentLinearAccScale=<1 1 1>
    componentLinearVelScale=<1 1 1>
    gravityOverride {z=-980}
    physicsBodyDefinitions [{
      boundBone {boneName=dyn_hairStrand_Side_A_01}
      boxExtents=<16 22 22>
      constraintSetup {angularLimitsMax=<0 10 25>, angularLimitsMin=<0 -15 -25>, angularTarget=<1 0 0>}
      localJointOffset=<1.5 0 0>
     }]
    simSpaceSettings {maxAngularAcceleration=100, maxAngularVelocity=20, simSpaceAngularAlpha=1}
   }
   n6 = Anim Dynamics - Bone: dyn_hairStrand_Side_B_01(ComponentPose=@n5, Alpha=1, GravityOverride="0,0,-980") {
    angularDampingOverride=1
    angularSpringConstant=10000
    bAngularSpring=true
    bOverrideAngularDamping=true
    bUseGravityOverride=true
    gravityScale=0.35
    simulationSpace=RootRelative
    alphaBoolBlend {blendOption=Linear}
    boundBone {boneName=dyn_hairStrand_Side_B_01}
    componentAppliedLinearAccClamp=<7500 7500 7500>
    componentLinearAccScale=<1 1 1>
    componentLinearVelScale=<1 1 1>
    gravityOverride {z=-980}
    physicsBodyDefinitions [{
      boundBone {boneName=dyn_hairStrand_Side_B_01}
      boxExtents=<18 22 22>
      constraintSetup {angularLimitsMax=<0 20 25>, angularLimitsMin=<0 -20 -25>, angularTarget=<1 0 0>}
      localJointOffset=<2.25 0 0>
     }]
    simSpaceSettings {maxAngularAcceleration=100, maxAngularVelocity=20, simSpaceAngularAlpha=1}
   }
   n7 = Anim Dynamics - Bone: dyn_hairStrand_Front_01(ComponentPose=@n6, Alpha=1, GravityOverride="0,0,-980") {
    angularDampingOverride=1
    angularSpringConstant=1000
    bAngularSpring=true
    bOverrideAngularDamping=true
    bUseGravityOverride=true
    gravityScale=0.35
    simulationSpace=RootRelative
    alphaBoolBlend {blendOption=Linear}
    boundBone {boneName=dyn_hairStrand_Front_01}
    componentAppliedLinearAccClamp=<3000 3000 3000>
    componentLinearAccScale=<1 1 1>
    componentLinearVelScale=<1 1 1>
    gravityOverride {z=-980}
    physicsBodyDefinitions [{
      boundBone {boneName=dyn_hairStrand_Front_01}
      boxExtents=<12 16 16>
      constraintSetup {angularLimitsMax=<0 40 25>, angularLimitsMin=<0 -20 -25>, angularTarget=<1 0 0>}
      localJointOffset=<1.5 0 0>
     }]
    simSpaceSettings {maxAngularAcceleration=100, maxAngularVelocity=20, simSpaceAngularAlpha=1}
   }
   n8 = Component To Local(ComponentPose=@n7)
   n9 = Anim Dynamics - Bone: dyn_hairStrand_top_01!(
    ComponentPose=Local To Component(LocalPose=Blend Poses by bool(BlendPose_0=@n8, BlendPose_1=@n4, BlendTime_0=0.5, BlendTime_1=0.5))
    Alpha=1
    GravityOverride:Vector="0,0,-980"
   ) -> (Pose) {
    angularDampingOverride=1
    angularSpringConstant=9000
    bAngularSpring=true
    bOverrideAngularDamping=true
    bUseGravityOverride=true
    gravityScale=0.35
    simulationSpace=RootRelative
    alphaBoolBlend {blendOption=Linear}
    boundBone {boneName=dyn_hairStrand_top_01}
    componentAppliedLinearAccClamp=<10000 10000 10000>
    componentLinearAccScale=<1 1 1>
    componentLinearVelScale=<1 1 1>
    gravityOverride {z=-980}
    physicsBodyDefinitions [{
      boundBone {boneName=dyn_hairStrand_top_01}
      boxExtents=<20 25 25>
      constraintSetup {angularLimitsMax=<10 20 25>, angularLimitsMin=<-10 -20 -25>, angularTarget=<1 0 0>}
      localJointOffset=<2 0 0>
     }]
    simSpaceSettings {maxAngularAcceleration=100, maxAngularVelocity=20, simSpaceAngularAlpha=1}
   }
   Save cached pose 'DynamicsCache'(Pose=Component To Local(ComponentPose=@n9)) {cacheName=DynamicsCache}
  ]}
 EventGraph {nodes [
   n1 = GetPartSkeletalMeshComponent()
   n2 = GetOverlayMaterial!(self:Mesh Component Object Reference=@n1, bGetMaterialSlot:bool=false, SlotIndex:int=0) -> (ReturnValue:Material Interface Object Reference)
   n3 = GetOrCreateDynamicOverlayMaterial(MeshComponent=@n1, MaterialName=String To Name(InString=GetDisplayName(Object=@n2)))
   > n4 = Set Outline Material(Outline Material:Material Instance Dynamic Object Reference=@n3) -> (Output_Get:Material Instance Dynamic Object Reference)
   > SetOverlayMaterial!(
    execute
    self:Mesh Component Object Reference=@n1
    NewOverlayMaterial:Material Interface Object Reference=@n4
    bSetMaterialSlot:bool=false
    SlotIndex:int=0
   ) -> (then)
   > n5 = Cast To FortPlayerPawn(Object=GetOwningActor())
   n6 = GetSocketLocation!(self:Scene Component Object Reference=Get Mesh(self=@n5), InSocketName:Name=attach_emote_01) -> (ReturnValue:Vector)
   SetVectorParameterValueOnMaterials!(execute, self:Mesh Component Object Reference=@n1, ParameterName:Name=Pelvis Position, ParameterValue:Vector=@n6) -> (then)
   n7 = BreakVector(InVec=@n6)
   n8 = Get Outline Material() -> (Outline Material:Material Instance Dynamic Object Reference)
   > n9 = SetVectorParameterValue!(execute, self=@n8, ParameterName=Pelvis Position, Value_R:float=@n7, Value_G:float=@n7.Y, Value_B:float=@n7.Z, Value_A:float=1) -> (then)
   n10 = Sequence(execute=[@n9.then, @n15.then, @n16.then])
   > n11 = "Gravity Override State Function (Parameter Asset) (Nativized)"(
    CharacterPartAnimInstance=Self-Reference()
    GameTimeInSeconds=GetGameTimeInSeconds()
    Parameters=&/M_MED_KneelReedy_FaceAcc_Parameters
   )
   > Set GravityOverride(GravityOverride=@n11)
   n12 = Get bIsMoving2D()
   n13 = Get bIsCrouching()
   Set looseHair(
    execute=@n10.then_1
    looseHair:bool=OR Boolean!(A=Get bIsSprinting(), B=Get bIsJumping(), C:bool=Get bIsFalling(), D:bool=AND Boolean(A=@n12, B=@n13)) -> (ReturnValue)
   ) -> (Output_Get:bool)
   n14 = GetAnimInstance(self=GetPartSkeletalMeshComponent())
   SetScalarParameterValue(self=@n8, ParameterName=MaterialCurveA, Value=GetCurveValue(self=@n14, CurveName=MaterialCurveA))
   > n15 = SetScalarParameterValue(self=@n8, ParameterName=MaterialCurveB, Value=GetCurveValue(self=@n14, CurveName=MaterialCurveB))
   n16 = Event BlueprintUpdateAnimation()
  ]}
]
variables [
 GravityOverride {type=Vector}
 looseHair {type=boolean}
 Outline Material {type=MaterialInstanceDynamic}
]
